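# Physics asset for the Spicy Soda consumable's skeletal mesh, defining collision bodies and joints for a cluster of six cans.
# A box body on Cans_Master acts as the hub, with a capsule on each dyn_can bone and a small capsule on Root.
# Each can is constrained directly to the hub with tight swing and twist limits, allowing only slight angular movement, with can 02 held tighter in swing.
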PhysicsAsset /Game/Athena/Items/Consumables/SpicySoda/Meshes/SK_SpicySoda_PhysicsAsset
bodies {
 Cans_Master [{shapeType=Box, extentX=7.5, extentY=7.5, extentZ=7.5, center=<0.0283 0.1226 0.5629>}]
 Root [{shapeType=Capsule, length=1, radius=1}]
 dyn_can_01 [{shapeType=Capsule, length=11.0514, radius=4.5724, center=<8.9377 0.3119 -0.0145>, rotation=<rot 0 -89.4016 -89.8647>}]
 dyn_can_02 [{shapeType=Capsule, length=11.0514, radius=4.5518, center=<7.9208 0.3035 0.0521>, rotation=<rot 0 90.6124 -89.9634>}]
 dyn_can_03 [{shapeType=Capsule, length=11.1027, radius=4.6423, center=<5.9131 0.2672 0.1302>, rotation=<rot 0 91.1201 -89.2736>}]
 dyn_can_04 [{shapeType=Capsule, length=11.1027, radius=4.6423, center=<1.6318 0.1834 0.076>, rotation=<rot 0 91.1201 -89.2736>}]
 dyn_can_05 [{shapeType=Capsule, length=11.0514, radius=4.6064, center=<2.8953 0.2296 0.0037>, rotation=<rot 0 -89.6183 -89.5197>}]
 dyn_can_06 [{shapeType=Capsule, length=11.0514, radius=4.5992, center=<3.8946 0.259 0.0605>, rotation=<rot 0 90.597 -89.8586>}]
}
constraints [
 {bone1Name=dyn_can_01, bone2Name=Cans_Master, swing1LimitDegrees=5, swing2LimitDegrees=5, twistLimitDegrees=1}
 {bone1Name=dyn_can_02, bone2Name=Cans_Master, swing1LimitDegrees=2, swing2LimitDegrees=2, twistLimitDegrees=1}
 {bone1Name=dyn_can_03, bone2Name=Cans_Master, swing1LimitDegrees=5, swing2LimitDegrees=5, twistLimitDegrees=1}
 {bone1Name=dyn_can_04, bone2Name=Cans_Master, swing1LimitDegrees=5, swing2LimitDegrees=5, twistLimitDegrees=1}
 {bone1Name=dyn_can_05, bone2Name=Cans_Master, swing1LimitDegrees=5, swing2LimitDegrees=5, twistLimitDegrees=1}
 {bone1Name=dyn_can_06, bone2Name=Cans_Master, swing1LimitDegrees=5, swing2LimitDegrees=5, twistLimitDegrees=1}
]
